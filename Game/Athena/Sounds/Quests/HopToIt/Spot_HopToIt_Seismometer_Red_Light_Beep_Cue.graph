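# Sound cue for the red-light beep of the seismometer in the HopToIt quest.
# A single wave player feeds the output directly, playing a wave named for both green and red.
# It uses a dedicated light-beep attenuation asset and the non-critical in-game UI sound class at reduced volume.
# A concurrency override allows one instance at a time, stopping the oldest when a new beep starts.
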SoundCue /Game/Athena/Sounds/Quests/HopToIt/Spot_HopToIt_Seismometer_Red_Light_Beep_Cue
root=WavePlayer_0
nodes [{id=WavePlayer_0, properties {soundWaveAssetPtr=&/Spot_HopToIt_Seismometer_Green_Red}}]
properties {
 attenuationSettings=&/Spot_Seismometer_Light_Beep_Att
 bOverrideConcurrency=true
 maxDistance=1500
 soundClassObject=&Fort_UI_InGame_NonCritical
 volumeMultiplier=0.3
 concurrencyOverrides {maxCount=1, resolutionRule=StopOldest}
}
